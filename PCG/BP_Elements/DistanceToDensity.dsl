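; Custom PCG Blueprint element, "Distance To Density", that turns each point's distance from a reference point into a gradient value, presumably its density.
; It converts spatial data from the first custom input pin to points, loops over them and keeps the target actor on the output.
; Each point's 2D or 3D distance is offset and scaled, then optionally clamped and inverted.
; It also derives a seed from each point's position.
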(fn ExecuteWithContext
  (params (InContext :type PCGContext%20Structure%20%28by%20ref%29) (Input :type PCGData%20Collection%20Structure%20%28by%20ref%29))
  (LocalVariables|SetContext InContext)
  (bind _returnvalue (PCG|Temporary|GetOriginalComponent (LocalVariables|GetContext)))
  (Variables|Hidden|SetPCGVolumeActor (Components|GetOwner _returnvalue))
  (bind _result (Utilities|Set|ToArray (PCG|Input&Output|CustomInputLabels self)))
  (bind _returnvalue_2 (Data|GetInputsbyPinLabel Input (Utilities|Array|Get(aref) _result)))
  (for _array_element _returnvalue_2
    (bind _aspcgspatial_data (Utilities|Casting|CastToPCGSpatialData (PCG|TaggedData|BreakTaggedData _array_element)))
    (bind _outdata (PCG|FlowControl|PointLoop (LocalVariables|GetContext) (SpatialData|ToPointDatawithContext _aspcgspatial_data (LocalVariables|GetContext)) 0))
    (Variables|/Script/PCG.PCGSpatialData|SetTargetActor (Variables|/Script/PCG.PCGSpatialData|GetTargetActor _aspcgspatial_data) _outdata)
    (Utilities|Array|Add (LocalVariables|GetOutTaggedData) (PCG|TaggedData|MakeTaggedData _outdata)))
  (return (Utilities|Struct|MakePCGDataCollection (LocalVariables|GetOutTaggedData))))

(fn NodeTitleOverride
  (return "Distance To Density"))

(fn PointLoopBody
  (params (InContext :type PCGContext%20Structure%20%28by%20ref%29) (InData :type PCGPoint%20Data%20Object%20Reference) (InPoint :type PCGPoint%20Structure%20%28by%20ref%29) (OutMetadata :type PCGMetadata%20Object%20Reference) (Iteration :type Integer64))
  (bind _transform (Utilities|Struct|BreakPCGPoint InPoint))
  (bind _location (.location _transform))
  (bind _referencepoint (Variables|Default|GetReferencePoint))
  (bind _returnvalue (Math|Vector|Distance(Vector) _referencepoint _location))
  (bind _returnvalue_2 (Math|Vector|Distance2D(Vector) _referencepoint _location))
  (bind _computedistance2d (Variables|Default|GetComputeDistance2D))
  (bind _returnvalue_3 (Math|Float|SelectFloat _returnvalue_2 _returnvalue _computedistance2d))
  (bind _gradientscale (Variables|Default|GetGradientScale))
  (bind _gradientoffset (Variables|Default|GetGradientOffset))
  (bind _returnvalue_4 (- _returnvalue_3 _gradientoffset))
  (bind _returnvalue_5 (Math|Float|SafeDivide _returnvalue_4 _gradientscale))
  (bind _returnvalue_6 (Math|Float|Clamp(Float) _returnvalue_5))
  (bind _returnvalue_7 (PCG|Helpers|ComputeSeedfromPosition _location))
  (if (Variables|Default|GetClampOutput)
    (if (Variables|Default|GetNegateClampedOutput)
      (bind _structout (Utilities|Struct|SetmembersinPCGPoint InPoint (- 1.0 _returnvalue_6) _returnvalue_7))
      (return _structout true)
      (else
        (bind _structout_2 (Utilities|Struct|SetmembersinPCGPoint InPoint _returnvalue_6 _returnvalue_7))
        (return _structout_2 true)))
    (else
      (bind _structout_3 (Utilities|Struct|SetmembersinPCGPoint InPoint _returnvalue_5 _returnvalue_7))
      (return _structout_3 true))))
