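; PCG Blueprint element "Export To PCG Asset" that captures its input as point data inside a PCGAsset element, passing the input through unchanged.
; When SaveOnNextExecution is set and the target PCG Settings are valid, it titles that settings' PCGAsset instance after them and clears its custom output pins.
; Each input converts to new point data owned by the settings, with flattened metadata and original tags, then gets stored as the asset's data collection.
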(fn ExecuteWithContext
  (params (InContext :type PCGContext%20Structure%20%28by%20ref%29) (Input :type PCGData%20Collection%20Structure%20%28by%20ref%29))
  (LocalVariables|SetContext InContext)
  (LocalVariables|SetInputCollection Input)
  (if (and (Utilities|IsValid (Variables|Default|GetPCG%20Settings) :!class /Script/Engine.KismetSystemLibrary) (Variables|Default|GetSaveOnNextExecution))
    (Development|PrintString (Utilities|String|Append "Exporting PCG Asset: " (Utilities|GetObjectName (Variables|Default|GetPCG%20Settings))))
    (bind _aspcgblueprint_settings (Utilities|Casting|CastToPCGBlueprintSettings (Variables|Default|GetPCG%20Settings)))
    (bind _aspcgasset (Utilities|Casting|CastToPCGAsset (Variables|/Script/PCG.PCGBlueprintSettings|GetBlueprintElementInstance _aspcgblueprint_settings)))
    (bind _output_get (LocalVariables|SetPCGAssetLocal _aspcgasset))
    (Variables|/PCG/BP_Elements/PCGAsset.PCGAsset_C|SetNodeTitle (Utilities|String|StringToName (Utilities|GetObjectName (Variables|Default|GetPCG%20Settings))) _output_get)
    (Variables|/Script/PCG.PCGBlueprintElement|SetCustomOutputPins (Utilities|Array|MakeArray) (LocalVariables|GetPCGAssetLocal))
    (for _array_element (Utilities|Struct|BreakPCGDataCollection (LocalVariables|GetInputCollection))
      (bind _aspcgspatial_data (Utilities|Casting|CastToPCGSpatialData (PCG|TaggedData|BreakTaggedData _array_element)))
      (LocalVariables|SetInputDataAsSpatialData _aspcgspatial_data)
      (bind _returnvalue (Game|ConstructObjectfromClass "/Script/PCG.PCGPointData" (Variables|Default|GetPCG%20Settings)))
      (bind _output_get_2 (LocalVariables|SetOutputPointData _returnvalue))
      (SpatialData|InitializefromData _output_get_2 (LocalVariables|GetInputDataAsSpatialData) 0 true true)
      (SpatialData|SetPoints (LocalVariables|GetOutputPointData) (SpatialData|GetPoints (SpatialData|ToPointDatawithContext (LocalVariables|GetInputDataAsSpatialData) (LocalVariables|GetContext))))
      (bind _returnvalue_2 (Metadata|MutableMetadata (LocalVariables|GetOutputPointData)))
      (PCG|Metadata|Advanced|Flatten _returnvalue_2)
      (Utilities|Array|Add (LocalVariables|GetOutputTaggedData) (PCG|TaggedData|MakeTaggedData (LocalVariables|GetOutputPointData) (.Tags (PCG|TaggedData|BreakTaggedData _array_element)))))
    (Variables|/PCG/BP_Elements/PCGAsset.PCGAsset_C|SetPCGDataCollection (Utilities|Struct|MakePCGDataCollection (LocalVariables|GetOutputTaggedData)) (LocalVariables|GetPCGAssetLocal))
    (return (LocalVariables|GetInputCollection))
    (else
      (return (LocalVariables|GetInputCollection)))))

(fn NodeTitleOverride
  (return "Export To PCG Asset"))
