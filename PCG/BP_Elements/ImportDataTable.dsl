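; PCG Blueprint element "Import Data Table" that turns the rows of a CSV-format points data table into PCG point data.
; Each row becomes a point with Y and Z swapped, position scaled by Scale, density from inverted UV Y, and row index as seed.
; Execute takes the target actor from the first spatial input, while ExecuteWithContext uses the PCG component's owner.
; Both append input settings to the output collection.
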(fn Execute
  (params (Input :type PCGData%20Collection%20Structure%20%28by%20ref%29))
  (bind _points_data_table (Variables|Default|GetPoints%20Data%20Table))
  (bind _output_get (LocalVariables|SetInputCollection Input))
  (bind _returnvalue (Data|GetAllSpatialInputs _output_get))
  (Utilities|FlowControl|Sequence
    (:then_0
      (bind _aspcgspatial_data (Utilities|Casting|CastToPCGSpatialData (PCG|TaggedData|BreakTaggedData (Utilities|Array|Get(aref) _returnvalue))))
      (LocalVariables|SetTargetActor (Variables|/Script/PCG.PCGSpatialData|GetTargetActor _aspcgspatial_data)))
    (:then_1
      (bind _outrownames (DataTable|GetDataTableRowNames _points_data_table))
      (for (_array_element _array_index) _outrownames
        (bind _returnvalue_2 (Utilities|GetDataTableRow _points_data_table _array_element))
        (Utilities|Array|Add (LocalVariables|GetOutPoints) (Utilities|Struct|MakePCGPoint (Math|Conversions|ToTransform(Vector) (* (Math|Vector|MakeVector (Utilities|Struct|BreakCSVPointFormat _returnvalue_2) (.Pz_8_488530224541EBF7D43F2F99AE2CCEAF (Utilities|Struct|BreakCSVPointFormat _returnvalue_2)) (.Py_7_462CEB644FCF8818190E0DA7A4233236 (Utilities|Struct|BreakCSVPointFormat _returnvalue_2))) (Math|Vector|MakeVector (Variables|Default|GetScale) (Variables|Default|GetScale) (Variables|Default|GetScale)))) (- 1.0 (.UVy_32_C10CBC5A416688DD4D1B1AB55E503EC2 (Utilities|Struct|BreakCSVPointFormat _returnvalue_2))) "-1.000000,-1.000000,-1.000000" "1.000000,1.000000,1.000000" "(X=1.000000,Y=1.000000,Z=1.000000,W=1.000000)" 0.5 _array_index))))
    (:then_3
      (bind _returnvalue_3 (Game|ConstructObjectfromClass "/Script/PCG.PCGPointData"))
      (SpatialData|SetPoints _returnvalue_3 (LocalVariables|GetOutPoints))
      (Variables|/Script/PCG.PCGSpatialData|SetTargetActor (LocalVariables|GetTargetActor) _returnvalue_3)
      (LocalVariables|SetOutputTaggedData (Utilities|Array|MakeArray (PCG|TaggedData|MakeTaggedData _returnvalue_3)))
      (bind _returnvalue_4 (Data|GetAllSettings (LocalVariables|GetInputCollection)))
      (Utilities|Array|AppendArray (LocalVariables|GetOutputTaggedData) _returnvalue_4)
      (return (Utilities|Struct|MakePCGDataCollection (LocalVariables|GetOutputTaggedData))))))

(fn ExecuteWithContext
  (params (InContext :type PCGContext%20Structure%20%28by%20ref%29) (Input :type PCGData%20Collection%20Structure%20%28by%20ref%29))
  (bind _points_data_table (Variables|Default|GetPoints%20Data%20Table))
  (bind _output_get (LocalVariables|SetInputCollection))
  (Data|GetAllSpatialInputs _output_get)
  (Utilities|FlowControl|Sequence
    (:then_1
      (bind _outrownames (DataTable|GetDataTableRowNames _points_data_table))
      (for (_array_element _array_index) _outrownames
        (bind _returnvalue (Utilities|GetDataTableRow _points_data_table _array_element))
        (Utilities|Array|Add (LocalVariables|GetOutPoints) (Utilities|Struct|MakePCGPoint (Math|Conversions|ToTransform(Vector) (* (Math|Vector|MakeVector (Utilities|Struct|BreakCSVPointFormat _returnvalue) (.Pz_8_488530224541EBF7D43F2F99AE2CCEAF (Utilities|Struct|BreakCSVPointFormat _returnvalue)) (.Py_7_462CEB644FCF8818190E0DA7A4233236 (Utilities|Struct|BreakCSVPointFormat _returnvalue))) (Math|Vector|MakeVector (Variables|Default|GetScale) (Variables|Default|GetScale) (Variables|Default|GetScale)))) (- 1.0 (.UVy_32_C10CBC5A416688DD4D1B1AB55E503EC2 (Utilities|Struct|BreakCSVPointFormat _returnvalue))) "-1.000000,-1.000000,-1.000000" "1.000000,1.000000,1.000000" "(X=1.000000,Y=1.000000,Z=1.000000,W=1.000000)" 0.5 _array_index))))
    (:then_3
      (bind _returnvalue_2 (Game|ConstructObjectfromClass "/Script/PCG.PCGPointData"))
      (SpatialData|SetPoints _returnvalue_2 (LocalVariables|GetOutPoints))
      (bind _returnvalue_3 (PCG|Temporary|GetComponent InContext))
      (Variables|/Script/PCG.PCGSpatialData|SetTargetActor (Components|GetOwner _returnvalue_3) _returnvalue_2)
      (LocalVariables|SetOutputTaggedData (Utilities|Array|MakeArray (PCG|TaggedData|MakeTaggedData _returnvalue_2)))
      (bind _returnvalue_4 (Data|GetAllSettings (LocalVariables|GetInputCollection)))
      (Utilities|Array|AppendArray (LocalVariables|GetOutputTaggedData) _returnvalue_4)
      (return (Utilities|Struct|MakePCGDataCollection (LocalVariables|GetOutputTaggedData))))))

(fn NodeTitleOverride
  (return "Import Data Table"))
