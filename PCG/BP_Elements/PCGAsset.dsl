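; A custom Blueprint PCG element that outputs a data collection stored in its own variable, acting as a PCG data asset node.
; ExecuteWithContext ignores its input and loops over the stored tagged data, casting each entry to spatial data and setting its target actor to the owner of the executing PCG component.
; It mutates that data in place rather than copying it, then returns it with original tags and pins.
; The node title comes from a variable; its color is black.
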(fn ExecuteWithContext
  (params (InContext :type PCGContext%20Structure%20%28by%20ref%29) (Input :type PCGData%20Collection%20Structure%20%28by%20ref%29))
  (bind _returnvalue (PCG|Temporary|GetComponent InContext))
  (for _array_element (Utilities|Struct|BreakPCGDataCollection (Variables|Default|GetPCGDataCollection))
    (bind _aspcgspatial_data (Utilities|Casting|CastToPCGSpatialData (PCG|TaggedData|BreakTaggedData _array_element)))
    (Variables|/Script/PCG.PCGSpatialData|SetTargetActor (Components|GetOwner _returnvalue) _aspcgspatial_data)
    (Utilities|Array|Add (LocalVariables|GetOutTaggedData) (PCG|TaggedData|MakeTaggedData _aspcgspatial_data (.Tags (PCG|TaggedData|BreakTaggedData _array_element)) (.Pin (PCG|TaggedData|BreakTaggedData _array_element)))))
  (return (Utilities|Struct|MakePCGDataCollection (LocalVariables|GetOutTaggedData))))

(fn NodeColorOverride
  (return "(R=0.000000,G=0.000000,B=0.000000,A=1.000000)"))

(fn NodeTitleOverride
  (return (Variables|Default|GetNodeTitle)))
