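; PCG Blueprint element that outputs point data with a point at the transform of each player controller's controlled pawn.
; It gathers all PlayerController actors and builds one point per controller, with a seed computed from the pawn's location.
; The points go into a new PCGPointData, targeted at the PCG component's owner and returned as the data collection.
; The Input collection is not read.
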(fn ExecuteWithContext
  (params (InContext :type PCGContext%20Structure%20%28by%20ref%29) (Input :type PCGData%20Collection%20Structure%20%28by%20ref%29))
  (bind _outactors (Actor|GetAllActorsOfClass "/Script/Engine.PlayerController"))
  (for _array_element _outactors
    (bind _asplayer_controller (Utilities|Casting|CastToPlayerController _array_element))
    (bind _returnvalue (PCG|Helpers|ComputeSeedfromPosition (.location (Transformation|GetActorTransform (Pawn|GetControlledPawn _asplayer_controller)))))
    (Utilities|Array|Add (LocalVariables|GetOutputPoints) (Utilities|Struct|MakePCGPoint (Transformation|GetActorTransform (Pawn|GetControlledPawn _asplayer_controller)) 1.0 "-1.000000,-1.000000,-1.000000" "1.000000,1.000000,1.000000" "(X=1.000000,Y=1.000000,Z=1.000000,W=1.000000)" 0.5 _returnvalue)))
  (bind _returnvalue_2 (Game|ConstructObjectfromClass "/Script/PCG.PCGPointData"))
  (bind _returnvalue_3 (PCG|Temporary|GetComponent InContext))
  (Variables|/Script/PCG.PCGSpatialData|SetTargetActor (Components|GetOwner _returnvalue_3) _returnvalue_2)
  (SpatialData|SetPoints _returnvalue_2 (LocalVariables|GetOutputPoints))
  (return (Utilities|Struct|MakePCGDataCollection (Utilities|Array|MakeArray (PCG|TaggedData|MakeTaggedData :Data _returnvalue_2)))))
