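; A custom PCG element that captures the scene over the target actor's bounds and outputs it as render target data for PCG graphs.
; RTRenderScene does the capture, configured by capture source, location, terrain and tag filters, texel density, format and post-process material.
; The texture initializes a PCGRenderTargetData placed at the actor's location and rotation, scaled by its bounds extent.
; CPU readback is optional, and the result leaves on a "RenderTarget" pin.
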(fn Execute
  (params (Input :type PCGData%20Collection%20Structure%20%28by%20ref%29))
  (bind _returnvalue (PCG|Temporary|GetTargetActor (PCG|Advanced|GetContext)))
  (bind _texturetarget (Default|RTRenderScene (Variables|Default|GetCapture%20Source) _returnvalue (Variables|Default|GetCaptureRelativeLocation) (.BoxExtent (Collision|GetActorBounds _returnvalue false false)) (Variables|Default|GetRenderTerrainOnly) (Variables|Default|GetTerrain%20Actor%20Class) (Variables|Default|GetActor%20Tag) (Variables|Default|GetTexels%20Per%20Meter) (Variables|Default|GetRender%20Target%20Format) (Variables|Default|GetPost%20Process%20Material)))
  (bind _returnvalue_2 (Game|ConstructObjectfromClass "/Script/PCG.PCGRenderTargetData"))
  (RenderTarget|Initialize _returnvalue_2 _texturetarget (Math|Transform|MakeTransform (.location (Transformation|GetActorTransform _returnvalue)) (.rotation (Transformation|GetActorTransform _returnvalue)) (.BoxExtent (Collision|GetActorBounds _returnvalue false false))) (not (Variables|Default|GetSkip%20Readback%20to%20CPU)) false)
  (return (Utilities|Struct|MakePCGDataCollection (Utilities|Array|MakeArray (PCG|TaggedData|MakeTaggedData :Data _returnvalue_2 :Pin "RenderTarget")))))
